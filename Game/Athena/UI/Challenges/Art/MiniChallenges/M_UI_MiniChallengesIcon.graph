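# Translucent UI material for mini challenge icons, tinting and shaping a mask texture, by default the AssaultDamage mini icon.
# The mask's red channel blends between two color parameters and, in parallel, between two alpha parameters.
# The green channel then scales that alpha from zero, so a single two-channel mask yields a two-tone icon with separate opacity per tone.
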Material /Game/Athena/UI/Challenges/Art/MiniChallenges/M_UI_MiniChallengesIcon
graph {
 properties {
  bUsedWithStaticMesh=false
  blendMode=BLEND_Translucent
  materialDomain=MD_UI
  pixelDepthOffsetMode=PDOM_Legacy
  usedShadingModels=Default Lit
 }
 nodes [
  n1 = TextureSampleParameter2D() {
   parameterName=TextureMask
   samplerType=SAMPLERTYPE_LinearColor
   texture=&/T_UI_Icon_Mini_AssaultDamage
  }
  LinearInterpolate(
   A=VectorParameter() {parameterName=RGB_2, defaultValue=<rgba 0 1 0.1208 1>}.RGB
   B=VectorParameter() {parameterName=RGB_1, defaultValue=<rgba 1 0 0.2098 1>}.RGB
   Alpha=@n1.R
  )
  LinearInterpolate(
   B=LinearInterpolate(A=ScalarParameter() {defaultValue=0.5, parameterName=Alpha_2}, B=ScalarParameter() {defaultValue=1, parameterName=Alpha_1}, Alpha=@n1.R)
   Alpha=@n1.G
  )
 ]
}
